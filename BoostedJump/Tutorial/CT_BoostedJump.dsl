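; Tutorial trigger for Boosted Jump that fires its activation trigger when the ability becomes available.
; On activate, it binds to changes of the Granted.BoostedJump.CanBeActivated gameplay tag on the owner controller's Ability System Component and stores the handle.
; A nonzero tag count in the callback calls FireActivationTrigger, and deactivation unbinds through the stored handle, so no polling is needed.
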(event EventOnActivate
  (bind _returnvalue (GameplayTags|BindEventWrappertoGameplayTagChanged (Ability|GetAbilitySystemComponent (CallFunction|GetOwnerController)) "(TagName=\"Granted.BoostedJump.CanBeActivated\")" (delegate Custom|OnBoostedJumpChanged)))
  (Variables|Default|SetGameplayEventHandle _returnvalue))

(event Custom|OnBoostedJumpChanged
  (params (Tag :type Gameplay%20Tag%20Structure%20%28by%20ref%29) (TagCount :type Integer))
  (if (!= TagCount 0)
    (CallFunction|FireActivationTrigger)))

(event EventOnDeactivate
  (GameplayTags|UnbindAllGameplayTagChangedEventWrappersforHandle (Variables|Default|GetGameplayEventHandle)))
